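# UI material function that turns a Blueprint-supplied StartTime into a timed animation value between 0 and 1.
# Time elapsed since StartTime, scaled by Speed, is wrapped by PlayDuration over Loops and remapped to a normalized phase, with TimeDirection reversing it.
# PlayType blends between ping-pong and one-way step shapes.
# A gate built from If nodes switches to resting values outside the play window, with zero Loops meaning infinite repeat.
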MaterialFunction /FortUILibrary/MaterialFunctions/Time/MF_UI_AnimationTimer
graph {
 properties {bExposeToLibrary=true, description=Blueprint controlled animation trigger, libraryCategoriesText [UI_ShaderLibrary]}
 nodes [
  n1 = NamedRerouteUsage()
  n2 = LinearInterpolate(
   A=NamedRerouteDeclaration(None=Constant()) {name=Starting Value, nodeColor=<rgba 0 0 0 1>}
   B=NamedRerouteDeclaration(None=Constant() {r=1}) {name=Ending Value, nodeColor=<rgba 0 0 0 1>}
   Alpha=NamedRerouteUsage()
  )
  n3 = LinearInterpolate(A=OneMinus(None=@n2), B=@n2, Alpha=OneMinus(None=NamedRerouteUsage()))
  n4 = LinearInterpolate(A=OneMinus(None=@n3), B=@n3, Alpha=NamedRerouteUsage())
  n5 = Multiply(A=Saturate(None=NamedRerouteUsage()), B=NamedRerouteUsage())
  n6 = FunctionInput() {
   bUsePreviewValueAsDefault=true
   desc=Obtained from BP
   description="In BP, use the node 'GetAccurateRealTime' and plug that value into here to start the animation."
   inputName=StartTime
   inputType=FunctionInput_Scalar
   sortPriority=8
   previewValue=<5358 0 0 0>
  }
  n7 = FunctionInput() {
   bUsePreviewValueAsDefault=true
   desc=Multiplier effects overall Play Duration
   description=Overall Speed Multiplier
   inputName=Speed
   inputType=FunctionInput_Scalar
   sortPriority=7
   previewValue=<1 0 0 1>
  }
  n8 = Multiply(A=Subtract(A=Time() {period=1}, B=NamedRerouteDeclaration(None=@n6) {name=StartTime, nodeColor=<rgba 0 0 0 1>}), B=@n7)
  n9 = Saturate(None=Divide(A=NamedRerouteDeclaration(None=@n8) {name=TimeOutput, nodeColor=<rgba 0 0 0 1>}, B=NamedRerouteUsage()))
  n10 = Constant() {r=1}
  n11 = Constant()
  n12 = LinearInterpolate(
   B=If(
    A=If(A=@n9, B=@n10, "A > B"=@n11, "A < B"=@n10)
    B=If(A=@n9, B=@n11, "A > B"=@n10, "A == B"=@n10, "A < B"=@n11)
    "A > B"=@n11
    "A == B"=@n10
    "A < B"=@n11
   )
   Alpha=Saturate(None=NamedRerouteUsage())
  ) {constA=1}
  n13 = NamedRerouteUsage()
  n14 = NamedRerouteUsage()
  n15 = FunctionInput() {
   bUsePreviewValueAsDefault=true
   desc="1 = Forwards, 0 = Backwards"
   description="1 = Forwards, 0 = Backwards"
   inputName=TimeDirection
   inputType=FunctionInput_Scalar
   sortPriority=3
   previewValue=<1 0 0 1>
  }
  n16 = NamedRerouteDeclaration(None=@n15) {name=TimeDirection, nodeColor=<rgba 0 0 0 1>}
  n17 = FunctionInput() {
   bUsePreviewValueAsDefault=true
   desc=In Seconds
   description="How long the entire animation should be in seconds (at 1x speed)"
   inputName=PlayDuration
   inputType=FunctionInput_Scalar
   sortPriority=2
   previewValue=<1 0 0 0>
  }
  n18 = Constant() {r=1}
  n19 = If(A=@n17, "A > B"=@n17, "A == B"=@n18, "A < B"=@n17)
  n20 = FunctionInput() {
   bUsePreviewValueAsDefault=true
   desc="0 = Infinity"
   description="The number of times the animation is repeated in the PlayDuration period. 0 = Infinity"
   inputName=Loops
   inputType=FunctionInput_Scalar
   sortPriority=6
   previewValue=<1 0 0 0>
  }
  n21 = Divide(A=@n19, B=If(A=NamedRerouteDeclaration(None=@n20) {name=Loops, nodeColor=<rgba 0 0 0 1>}, "A > B"=@n20, "A == B"=@n18, "A < B"=@n20))
  n22 = MaterialFunctionCall(Input=Fmod(A=NamedRerouteUsage(), B=@n21), Input Low=Constant(), Input High=@n21) {materialFunction=&RemapValue_0-1}
  n23 = Add(A=Multiply(A=@n22.Result, B=@n16), B=Multiply(A=OneMinus(None=@n22.Result), B=OneMinus(None=@n16)))
  n24 = NamedRerouteDeclaration(None=Saturate(None=Abs(None=@n23))) {name=LoopIteration, nodeColor=<rgba 0 0 0 1>}
  n25 = Add(A=Frac(None=@n24), B=Constant() {r=-0.5})
  n26 = Saturate(None=Multiply(A=Abs(None=@n25)) {constB=2})
  n27 = Subtract(A=OneMinus(None=@n26), B=Saturate(None=Floor(None=@n24)))
  n28 = NamedRerouteUsage()
  n29 = LinearInterpolate(A=@n13, B=@n14, Alpha=LinearInterpolate(A=Saturate(None=@n27), B=@n26, Alpha=OneMinus(None=@n28)))
  n30 = LinearInterpolate(A=@n13, B=@n14, Alpha=NamedRerouteUsage())
  n31 = FunctionInput() {
   bUsePreviewValueAsDefault=true
   desc="0 = PingPong\r\n1 = Step"
   description="0 = PingPong (Loops back to initial value), 1= Step (Moves from 0->1 or vice versa)"
   inputName=PlayType
   inputType=FunctionInput_Scalar
   sortPriority=4
   previewValue=<1 0 0 0>
  }
  n32 = NamedRerouteDeclaration(None=@n31) {name=PlayType, nodeColor=<rgba 0 0 0 1>}
  FunctionOutput(None=LinearInterpolate(A=LinearInterpolate(A=@n1, B=@n4, Alpha=@n5), B=LinearInterpolate(A=@n29, B=@n30, Alpha=@n32), Alpha=@n12))
  MaterialFunctionCall(Number=Time()) {materialFunction=&DebugScalarValues}
  MaterialFunctionCall!(X range Y range, Gradient Height, ShowGrid?, Thickness, Color, "f(x)"=NamedRerouteUsage()) {materialFunction=/Engine/Functions/Engine_MaterialFunctions01/Debug/PlotFunctionOnGraph}
  NamedRerouteDeclaration(None=@n12) {name=Gate Progress, nodeColor=<rgba 0 0 0 1>}
  NamedRerouteDeclaration(None=@n9) {name=CompleteDuration, nodeColor=<rgba 0 0 0 1>}
  NamedRerouteDeclaration(None=@n19) {name=PlayDuration_Safe, nodeColor=<rgba 0 0 0 1>}
  NamedRerouteUsage()
 ]
}
