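# Parent material for spring event eggs, using either simple parameter-driven surface attributes or the weapon master function via a static switch.
# An emissive highlight combines noisy Fresnel glow with an animated, camera-oriented sweep offset by object position, and masks darken base color beneath it.
# World position offset wobbles the egg around its local-bounds bottom, with an animation timer fed by custom primitive data.
# Results pass through MF_CreativeMoveTool.
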Material /SpringEvent/Items/SharedMaterials/M_SpringEvent_Egg_Parent
graph {
 properties {
  bAutomaticallySetUsageInEditor=false
  bUseMaterialAttributes=true
  bUsedWithClothing=true
  bUsedWithMeshDeformer=true
  bUsedWithMorphTargets=true
  bUsedWithNanite=true
  bUsedWithSkeletalMesh=true
  pixelDepthOffsetMode=PDOM_Legacy
  usedShadingModels=Default Lit
  assetUserData ["<self>:GLTFMaterialExportOptions_0"]
 }
 nodes [
  Constant() {r=1}
  Constant()
  n1 = ScalarParameter() {defaultValue=0.99, parameterName=Tight Sweep Size}
  n2 = ScalarParameter() {defaultValue=0.958, parameterName=Wide Sweep Size}
  n3 = CameraPositionWS()
  n4 = WorldPosition()
  LinearInterpolate(
   A=@n1
   B=@n2
   Alpha=InverseLinearInterpolate(
    A=ScalarParameter() {defaultValue=100, parameterName=Close Distance}
    B=ScalarParameter() {defaultValue=4000, parameterName="Far  Distance"}
    Value=Distance(A=@n3, B=@n4.XYZ)
   )
  ) {desc=distance width}
  n5 = SetMaterialAttributes!(
   MaterialAttributes
   Base Color=VectorParameter() {parameterName=Basic Color, defaultValue=<rgba 1 0.5683 0.2174 1>}.RGB
   Metallic=ScalarParameter() {defaultValue=1, parameterName=Metallic}
   Specular=ScalarParameter() {defaultValue=1, parameterName=Specular}
   Roughness=ScalarParameter() {defaultValue=0.28, parameterName=Roughnmess}
  ) {attributeSetTypes [
    #1
    #2
    #3
    #4
   ]}
  n6 = GetMaterialAttributes(None=StaticSwitchParameter(
    True=@n5
    False=MaterialFunctionCall!(In) {materialFunction=/Game/Packages/Fortress_SharedMaterials/Base_Material_Functions/BaseCustomizationFunctions/MF_FN_Weapon_MASTER}.Material Result
   ) {parameterName=UseSimpleMaterial}) {attributeGetTypes [#5, #1]}
  n7 = NamedRerouteDeclaration(None=@n6.Base Color) {name=Diffuse, nodeColor=<rgba 1 0 0.3047 1>}
  n8 = Max(A=NamedRerouteUsage(), B=NamedRerouteUsage())
  n9 = ScalarParameter() {defaultValue=4, parameterName=Darkening Multiplier}
  n10 = OneMinus(None=Saturate(None=Multiply(A=@n8, B=@n9)))
  n11 = NamedRerouteUsage()
  MaterialFunctionCall(MA In=SetMaterialAttributes!(
    MaterialAttributes=@n6.MaterialAttributes
    Emissive Color=Add(A=@n6.Emissive Color, B=@n11) {constB=0}
    World Position Offset=NamedRerouteUsage()
    Base Color=Multiply(A=@n7, B=@n10)
   ) {attributeSetTypes [#5, #6, #1]}) {materialFunction=&MF_CreativeMoveTool}
  n12 = ScalarParameter() {defaultValue=6, parameterName=FresnelExponent}
  n13 = Multiply(A=TextureCoordinate(), B=ScalarParameter() {defaultValue=9, parameterName=Fresnel Normal Noise Size})
  n14 = Transform(None=MaterialFunctionCall(
    Normal=TextureSample(UVs=@n13) {samplerType=SAMPLERTYPE_Normal, texture=/Game/Weapons/WeaponSkins/Textures/T_TileNoise_03_NRM}.RGB
    Flatness=ScalarParameter() {defaultValue=0.9, parameterName=FlattenDetailedNormal}
   ) {materialFunction=&FlattenNormal}.Result)
  n15 = Multiply(
   A=NamedRerouteDeclaration(None=Fresnel(ExponentIn=@n12, Normal=@n14) {baseReflectFraction=0}) {name=Glow Fresnel Mask, nodeColor=<rgba 1 0 0.6797 1>}
   B=VectorParameter() {parameterName=GlowColour, defaultValue=<rgba 3 2.6258 0.8273 0>}.RGB
  )
  n16 = Constant()
  n17 = WorldPosition()
  n18 = ObjectPositionWS()
  n19 = Normalize(VectorInput=Transform(None=VectorParameter() {parameterName=SweepDirectionCamSpace, defaultValue=<rgba -0.5 0.45 -1 0>}.RGB) {transformSourceType=TRANSFORMSOURCE_Camera})
  n20 = DotProduct(A=Multiply(A=Subtract(A=@n17.XYZ, B=@n18), B=ScalarParameter() {defaultValue=0.01, parameterName=WorldTileScale}), B=@n19)
  n21 = Time()
  n22 = NamedRerouteUsage()
  n23 = Constant() {r=1}
  n24 = Multiply(A=ComponentMask(None=Desaturation(None=@n22, Fraction=@n23)) {r=true}, B=ScalarParameter() {defaultValue=0.15, parameterName=Dialation})
  n25 = ObjectPositionWS()
  n26 = Divide(A=ComponentMask(None=@n25) {g=true}) {constB=3.14}
  n27 = Divide(A=ComponentMask(None=@n25) {r=true}) {constB=21.49}
  n28 = Sine(None=Frac(None=Add(A=@n26, B=@n27)))
  n29 = Panner(
   Coordinate=ComponentMask(None=Multiply(A=@n20, B=@n19)) {b=true, r=true}
   Time=Add(A=Add(A=@n21, B=@n24), B=NamedRerouteDeclaration(None=@n28) {name=Random Value, nodeColor=<rgba 0 0.1562 1 1>})
   Speed=ScalarParameter() {defaultValue=0.3, parameterName=SweepSpeed}
  ) {speedX=0.2, speedY=0.2}
  n30 = OneMinus(None=Frac(None=ComponentMask(None=@n29) {g=true}))
  n31 = Max(
   A=InverseLinearInterpolate(A=@n1, Value=@n30) {constA=0.99}
   B=Multiply(A=InverseLinearInterpolate(A=@n2, Value=@n30) {constA=0.99}, B=ScalarParameter() {defaultValue=0.1, parameterName=Wide Sweep Multiplier})
  )
  n32 = Multiply(
   A=NamedRerouteDeclaration(None=Saturate(None=@n31)) {name=Sweep Mask, nodeColor=<rgba 0.875 1 0 1>}
   B=VectorParameter() {parameterName=SweepColour, defaultValue=<rgba 8 7.0022 2.2061 0>}.RGB
  ) {constA=1}
  n33 = Constant()
  NamedRerouteDeclaration(None=Max(
    A=StaticSwitchParameter(True=@n15, False=@n16) {defaultValue=true, parameterName=UseOuterGlow}
    B=StaticSwitchParameter(True=@n32, False=@n33) {defaultValue=true, parameterName=UseSweep}
   )) {name=Highlight, nodeColor=<rgba 0 1 0.1797 1>}
  n34 = Multiply(A=Add(A=NamedRerouteUsage(), B=Time()), B=ScalarParameter() {defaultValue=6.12, parameterName=Param_1})
  MaterialExpressionAppend3Vector_0 = MaterialXAppend3Vector!(A=Sine(None=@n34), B=Cosine(None=@n34), C=Sine(None=@n34) {period=0.91})
  n35 = ComponentMask(None=MaterialFunctionCall() {materialFunction=&ObjectLocalBounds}.Local Bounds Minimum) {b=true}
  n36 = RotateAboutAxis(
   NormalizedRotationAxis=Normalize(VectorInput=Normalize(VectorInput=@MaterialExpressionAppend3Vector_0))
   RotationAngle=ScalarParameter() {defaultValue=0.01}
   PivotPoint=TransformPosition(None=MaterialFunctionCall(Z=@n35) {materialFunction=&MakeFloat3}.Result) {transformType=TRANSFORMPOSSOURCE_World}
   Position=WorldPosition().XYZ
  )
  NamedRerouteDeclaration(None=LinearInterpolate(B=Add(A=@n36), Alpha=NamedRerouteUsage())) {name=WPO, nodeColor=<rgba 1 0.4922 0 1>}
  NamedRerouteDeclaration(None=MaterialFunctionCall!(
    PlayDuration=ScalarParameter() {bUseCustomPrimitiveData=true, parameterName=Ease, primitiveDataIndex=1}
    TimeDirection=ScalarParameter() {bUseCustomPrimitiveData=true, parameterName=Direction, primitiveDataIndex=2}
    PlayType
    Loops
    Speed
    StartTime=ScalarParameter() {bUseCustomPrimitiveData=true, parameterName=AnimStartTime}
   ) {materialFunction=/FortUILibrary/MaterialFunctions/Time/MF_UI_AnimationTimer}.Result) {name=Anim Value, nodeColor=<rgba 0.9922 0 1 1>}
  PreSkinnedPosition()
  ScalarParameter() {parameterName=TimeStart}
  Time()
 ]
}
